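# Masked two-sided parent material for the IndoorLaceScore female outfit's patterned sleeves, built between the CharacterParent PreFX and PostFX functions.
# A tiled grayscale pattern texture and Fresnel-driven width and fade build an opacity mask, with a temporally dithered variant that is zero in shadow passes.
# The IsBackface static switch selects a variant that masks to backfaces and uses dimmed sky-light environment color blended with PatternColor as base and emissive.
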Material /BRCosmetics/Characters/Player/Female/Medium/Bodies/F_MED_IndoorLaceScore/Materials/Proto/M_IndoorLaceScore_Sleeve_Masked_Parent
graph {
 properties {
  bAutomaticallySetUsageInEditor=false
  bUseMaterialAttributes=true
  bUsedWithMeshDeformer=true
  bUsedWithMorphTargets=true
  bUsedWithSkeletalMesh=true
  blendMode=BLEND_Masked
  pixelDepthOffsetMode=PDOM_Legacy
  shadingModel=MSM_FromMaterialExpression
  twoSided=true
  usedShadingModels=Default Lit | Subsurface | Clear Coat | Subsurface Profile | Cloth
 }
 nodes [
  n1 = MaterialFunctionCall() {materialFunction=&MF_CharacterParent_2023_PreFX}
  n2 = GetMaterialAttributes(None=@n1.Result) {attributeGetTypes [#1]}
  n3 = Multiply(A=TwoSidedSign()) {constB=-1}
  n4 = Multiply(A=CameraVectorWS()) {constB=-1}
  n5 = Constant() {r=0.1}
  n6 = VectorParameter() {group=IndoorLaceScore - Sleeves, parameterName=PatternColor, defaultValue=<rgba 0.016 0.0152 0.0168 1>}
  n7 = Multiply(A=LinearInterpolate(A=SkyLightEnvMapSample(Direction=@n4, Roughness=@n5), B=@n6.RGB, Alpha=NamedRerouteUsage())) {constB=0.3}
  n8 = SetMaterialAttributes!(
   MaterialAttributes=@n1.Result
   Opacity Mask=NamedRerouteUsage()
   Base Color=@n6.RGB
   Metallic
   Specular=ScalarParameter() {defaultValue=0.5, group=IndoorLaceScore - Sleeves, parameterName=PatternSpecular}
   Roughness=ScalarParameter() {defaultValue=0.5, group=IndoorLaceScore - Sleeves, parameterName=PatternRoughess}
   Shading Model=ShadingModel()
  ) {attributeSetTypes [
    #2
    #1
    #3
    #4
    #5
    #6
   ]}
  MaterialFunctionCall(
   Material=StaticSwitchParameter(
    True=SetMaterialAttributes!(MaterialAttributes=@n2.MaterialAttributes, Opacity Mask=@n3, Base Color=@n7, Emissive Color=@n7) {attributeSetTypes [#2, #1, #7]}
    False=@n8
   ) {group=IndoorLaceScore - Sleeves, parameterName=IsBackface}
   M=@n1.M
   Pre Skinned Local Position=@n1.M
   Distance Blend=@n1.M
   Fresnel=@n1.M
  ) {materialFunction=&MF_CharacterParent_2023_PostFX}
  n9 = TextureSample(UVs=Multiply(A=TextureCoordinate(), B=ScalarParameter() {defaultValue=315, group=IndoorLaceScore - Sleeves, parameterName=PatternTiling})) {
   samplerType=SAMPLERTYPE_LinearGrayscale
   texture=/BRCosmetics/Characters/Player/Female/Medium/Bodies/F_MED_IndoorLaceScore/Materials/T_IndoorLaceScore_Pattern
  }
  NamedRerouteDeclaration(None=@n9.R) {name=PatternTexture, nodeColor=<rgba 0 1 0.2031 1>}
  n10 = ScalarParameter() {defaultValue=0.4, group=IndoorLaceScore - Sleeves, parameterName=PatternWidth_Center}
  n11 = ScalarParameter() {group=IndoorLaceScore - Sleeves, parameterName=PatternWidth_Edges}
  n12 = Fresnel(ExponentIn=ScalarParameter() {defaultValue=1, group=IndoorLaceScore - Sleeves, parameterName=FresnelExponent}) {baseReflectFraction=0}
  n13 = Max(
   A=SmoothStep(
    Min=LinearInterpolate(A=@n10, B=@n11, Alpha=@n12)
    Max=ScalarParameter() {defaultValue=0.5, group=IndoorLaceScore - Sleeves, parameterName=PatternHardness}
    Value=NamedRerouteUsage()
   )
   B=SmoothStep(
    Min=ScalarParameter() {group=IndoorLaceScore - Sleeves, parameterName=FresnelFadeMin}
    Max=ScalarParameter() {defaultValue=1, group=IndoorLaceScore - Sleeves, parameterName=FresnelFadeMax}
    Value=@n12
   ) {constMin=0.5}
  )
  NamedRerouteDeclaration(None=ShadowReplace(
    Default=MaterialFunctionCall!(Alpha Threshold=@n13, Random) {materialFunction=/Game/TimeOfDay/CloudsAndStorm/Materials/MaterialFunctions/DitherTemporalAABlue}.Result
    Shadow=Constant()
   )) {name=DitheredOpacityMask, nodeColor=<rgba 0 0.6016 1 1>}
  NamedRerouteDeclaration(None=@n13) {name=OpacityMask, nodeColor=<rgba 0.1953 0 1 1>}
  NamedRerouteUsage()
  Power(
   Base=ChannelMaskParameter(None=@n1.Tech Art Mask) {group=IndoorLaceScore - Sleeves, parameterName=PatternOpacityMask}.R
   Exp=ScalarParameter() {defaultValue=1, parameterName=Exp}
  )
 ]
}
